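; Building gameplay actor for the TidalCrane ToyFox boss base that knocks back players and vehicles in its way on spawn, preventing penetration.
; Once playlist data is ready, the server runs an async sphere overlap, cancelled if the actor is being destroyed.
; Hit actors receive a knockback gameplay effect, and player pawns temporarily ignore collision with the base.
; After completion and a delay, replicated bCollisionEnable re-enables collision through its OnRep, with hotfixable radius and timings.
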(event EventBeginPlay
  (bind _hotfix_basespawnknockbackradius (Variables|Hotfix|GetHotfix_BaseSpawnKnockbackRadius))
  (bind _returnvalue (ScalableFloat|GetValueatLevel _hotfix_basespawnknockbackradius))
  ;[0.004879 0.15 0 0.5] BeginPlay: Goal here is to knockback players and vehicles that are in the way when this base spawns to avoid penetration issues.
  (|CallParentFunction)
  (Fortnite|FortShippingLog self "BeginPlay: Invoked.")
  (Networking|SwitchHasAuthority
    (:Authority
      (Networking|FlushNetDormancy)
      (Class|FortAsyncActionCurrentPlaylistReady|PlaylistDataReadyAsync
        (:OnReady
          (Class|AsyncOverlapQueue|AsyncSphereOverlapActors (Transformation|GetActorLocation) _returnvalue (Utilities|Array|MakeArray self) (EventDispatchers|CreateEvent :!event HandleAsyncOverlapSpawnKnocbackShouldCancel) (EventDispatchers|CreateEvent :!event HandleAsyncOverlapSpawnKnocbackHitActor) (EventDispatchers|CreateEvent :!event Custom|HandleAsyncOverlapSpawnKnockbackComplete) "ECC_GameTraceChannel3")
          (if (Variables|Debug|GetbDebug)
            (Rendering|Debug|DrawDebugSphere (Transformation|GetActorLocation) _returnvalue 12 "(R=0.000000,G=1.000000,B=0.723820,A=1.000000)" 1e+01 3.0)))))))

(event Custom|HandleAsyncOverlapSpawnKnockbackComplete
  (params (WasCancelled :type Boolean))
  ;[0 0.15 0.07954 0.5] HandleAsyncOverlapSpawnKnockbackComplete: Once the knockback is complete, wait a short duration before re-enabling collision. Ensures knocked back vehicles have the chance to get out of the way.
  (Networking|SwitchHasAuthority
    (:Authority
      (Utilities|FlowControl|Delay (ScalableFloat|GetCachedScalableFloatValue (Variables|Hotfix|GetHotfix_BaseSpawnKnockbackDisableSelfCollisionDuration)))
      (Variables|Transient|SetbCollisionEnable true)
      (Networking|FlushNetDormancy))))

(fn HandleAsyncOverlapSpawnKnocbackHitActor
  (params (HitActor :type Actor%20Object%20Reference))
  ;[0 0.15 0.112 0.5] HandleAsyncOverlapSpawnKnocbackHitActor: Apply knockback gameplay effect to the overlapped actor.
  (Networking|SwitchHasAuthority
    (:Authority
      (Utilities|IsValid HitActor
        (:Is%20Valid
          (Fortnite|FortShippingLog self (Utilities|String|Append "HandleAsyncOverlapSpawnKnocbackHitActor: Hit Actor: " (Utilities|GetDisplayName HitActor)))
          (Fortnite|ApplyGameplayEffectsToActor HitActor (Utilities|Array|MakeArray (Variables|Settings|GetBaseSpawnKnockbackGameplayEffectClass)) (Transformation|GetActorLocation) false 0 (Ability|GetFortAbilitySystemComponent self))
          (bind _asfort_player_pawn (Utilities|Casting|CastToFortPlayerPawn HitActor))
          (Collision|AddIgnoredActor _asfort_player_pawn self (ScalableFloat|GetCachedScalableFloatValue (Variables|Hotfix|GetHotfix_BaseSpawnKnockbackPawnIgnoreCollisionDuration))))))))

(fn HandleAsyncOverlapSpawnKnocbackShouldCancel
  (return (Game|IsActorBeingDestroyed)))

(fn OnRep_bCollisionEnable
  (Collision|SetActorEnableCollision true))

(fn ConstructionScript
  (|CallParentFunction))
